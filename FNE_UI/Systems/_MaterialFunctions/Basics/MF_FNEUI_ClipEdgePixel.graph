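# UI material function, exposed to the library, that clips the last edge pixel of an alpha value.
# Two Steps against 0 and 1 are subtracted to mask UVs, defaulting to normalized user-interface UVs, to that range.
# The mask's R component is multiplied by itself and the Alpha input, which defaults to 1.
# The G component is unused, so only the U axis affects the result.
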MaterialFunction /FNE_UI/Systems/_MaterialFunctions/Basics/MF_FNEUI_ClipEdgePixel
graph {
 properties {bExposeToLibrary=true, description=Clip the last pixel on the edge.}
 nodes [
  n1 = Constant()
  n2 = FunctionInput(Preview=MaterialFunctionCall() {materialFunction=&GetUserInterfaceUV}.Normalized UV) {bUsePreviewValueAsDefault=true, description=Set the UV, inputName=UVs, inputType=FunctionInput_Vector2}
  n3 = Constant() {r=1}
  n4 = MaterialFunctionCall(Float2=Subtract(A=Step(Y=@n1, X=@n2), B=Step(Y=@n3, X=@n2))) {materialFunction=&BreakOutFloat2Components}
  n5 = FunctionInput(Preview=Constant() {r=1}) {bUsePreviewValueAsDefault=true, description=Set the alpha, inputName=Alpha, inputType=FunctionInput_Scalar, sortPriority=1}
  FunctionOutput(None=Multiply(A=Multiply(A=@n4.R, B=@n4.R), B=@n5)) {description=Returns a clipped alpha}
 ]
}
