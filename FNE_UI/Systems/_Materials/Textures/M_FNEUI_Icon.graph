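# Translucent UI material for FNE UI icons, drawing a widget-supplied texture with button-state color and opacity.
# A shared state function blends idle, focused, hovered, pressed and disabled values using activation and transition scalars.
# Edge-clipped texture alpha and a brightness mask, sampling only red and blue, feed transparency variants.
# Widget-controlled flags select tinting, black-and-white, transparent black or pure-black-only clearing through lerps linked by named reroutes.
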Material /FNE_UI/Systems/_Materials/Textures/M_FNEUI_Icon
graph {
 properties {blendMode=BLEND_Translucent, materialDomain=MD_UI, pixelDepthOffsetMode=PDOM_Legacy, usedShadingModels=Default Lit}
 nodes [
  n1 = ScalarParameter() {
   desc="Editing this parameter will force the setting and override the blueprint setting.\r\n"
   group="[0] Do Not Edit"
   parameterName=Is TINT Icon
   sliderMax=1
   sortPriority=0
  }
  Floor(None=@n1)
  LinearInterpolate(A=NamedRerouteUsage(), B=NamedRerouteUsage(), Alpha=NamedRerouteUsage())
  LinearInterpolate(A=NamedRerouteUsage(), B=NamedRerouteUsage(), Alpha=NamedRerouteUsage())
  n2 = LinearInterpolate(A=NamedRerouteUsage(), B=NamedRerouteUsage(), Alpha=NamedRerouteUsage())
  LinearInterpolate(A=@n2, B=Multiply(A=@n2, B=NamedRerouteUsage()), Alpha=NamedRerouteUsage()) {constAlpha=1}
  n3 = NamedRerouteUsage()
  n4 = NamedRerouteUsage()
  LinearInterpolate(
   A=LinearInterpolate(A=@n3, B=Multiply(A=@n3, B=@n4), Alpha=NamedRerouteUsage()) {constAlpha=1}
   B=LinearInterpolate(A=NamedRerouteUsage(), B=NamedRerouteUsage(), Alpha=NamedRerouteUsage())
   Alpha=NamedRerouteUsage()
  )
  NamedRerouteDeclaration(None=Saturate(None=@n1)) {name=Is TINT Icon, nodeColor=<rgba 0 0 0 1>}
  n5 = ScalarParameter() {
   desc="Editing this parameter will force the setting and override the blueprint setting.\r\n"
   group="[0] Do Not Edit"
   parameterName=Clear Pure Black Only
   sliderMax=1
   sortPriority=2
  }
  NamedRerouteDeclaration(None=Saturate(None=Floor(None=@n5))) {name=Is Pure Black Transparent Only, nodeColor=<rgba 0 0 0 1>}
  n6 = TextureSampleParameter2D() {
   desc="DO NOT EDIT : This texture parameter is exposed to let the widget blueprint communicate with the material, and apply the Icon Texture set in the widget blueprint. \r\n\r\nEditing this parameter will force the setting and override the blueprint setting.\r\n\r\n"
   group="[0] Do Not Edit"
   parameterName=Dynamic Icon Texture
   sortPriority=0
  }
  NamedRerouteDeclaration(None=@n6.RGB) {name=Icon RGB, nodeColor=<rgba 1 0 0.4922 1>}
  n7 = VectorParameter() {group=State Colors, parameterName=Color and Opacity, sortPriority=0, defaultValue=<rgba 1 1 1 1>}
  n8 = ScalarParameter() {
   desc="0 (<1) = Not active\r\n1 = Active\r\n\r\nWhen Active, The following parameters will be applied for the active state.\r\n\r\nWhen Not active, whatever the state parameter value, it will not be applied.\r\n\r\nUseful to set the states you are using only.\r\n"
   group=State Activations
   parameterName=Activate Focused State
   sliderMax=1
   sortPriority=2
  }
  n9 = VectorParameter() {group=State Colors, parameterName=Focused Color, sortPriority=2, defaultValue=<rgba 1 1 1 1>}
  n10 = ScalarParameter() {
   desc="WARNING : \r\nThe name of this parameter must not be edited. It is used in many widget parenting. Changing its name will brake the transition behavior.\r\n"
   group=State Transitions
   parameterName=Trans_Focused
   sliderMax=1
   sortPriority=1
  }
  n11 = ScalarParameter() {
   desc="0 (<1) = Not active\r\n1 = Active\r\n\r\nWhen Active, The following parameters will be applied for the active state.\r\n\r\nWhen Not active, whatever the state parameter value, it will not be applied.\r\n\r\nUseful to set the states you are using only.\r\n"
   group=State Activations
   parameterName=Activate Hovered State
   sliderMax=1
   sortPriority=1
  }
  n12 = VectorParameter() {group=State Colors, parameterName=Hovered Color, sortPriority=1, defaultValue=<rgba 1 1 1 1>}
  n13 = ScalarParameter() {
   desc="WARNING : \r\nThe name of this parameter must not be edited. It is used in many widget parenting. Changing its name will brake the transition behavior.\r\n"
   group=State Transitions
   parameterName=Trans_Hovered
   sliderMax=1
   sortPriority=0
  }
  n14 = ScalarParameter() {
   desc="0 (<1) = Not active\r\n1 = Active\r\n\r\nWhen Active, The following parameters will be applied for the active state.\r\n\r\nWhen Not active, whatever the state parameter value, it will not be applied.\r\n\r\nUseful to set the states you are using only.\r\n"
   group=State Activations
   parameterName=Activate Pressed State
   sliderMax=1
   sortPriority=3
  }
  n15 = VectorParameter() {group=State Colors, parameterName=Pressed Color, sortPriority=3, defaultValue=<rgba 1 1 1 1>}
  n16 = ScalarParameter() {
   desc="WARNING : \r\nThe name of this parameter must not be edited. It is used in many widget parenting. Changing its name will brake the transition behavior.\r\n"
   group=State Transitions
   parameterName=Trans_Pressed
   sliderMax=1
   sortPriority=2
  }
  n17 = ScalarParameter() {
   desc="0 (<1) = Not active\r\n1 = Active\r\n\r\nWhen Active, The following parameters will be applied for the active state.\r\n\r\nWhen Not active, whatever the state parameter value, it will not be applied.\r\n\r\nUseful to set the states you are using only.\r\n"
   group=State Activations
   parameterName=Activate Disabled State
   sliderMax=1
   sortPriority=4
  }
  n18 = VectorParameter() {group=State Colors, parameterName=Disabled Color, sortPriority=4, defaultValue=<rgba 1 1 1 1>}
  n19 = ScalarParameter() {
   desc="WARNING : \r\nThe name of this parameter must not be edited. It is used in many widget parenting. Changing its name will brake the transition behavior.\r\n"
   group=State Transitions
   parameterName=TransDisabled
   sliderMax=1
   sortPriority=3
  }
  n20 = MaterialFunctionCall!(
   Idle=@n7.RGB
   Idle Opacity=@n7.A
   Is Focused Active=@n8
   Focused=@n9.RGB
   Focused Opacity=@n9.A
   Trans_Focused=@n10
   Is Hovered Active=@n11
   Hovered=@n12.RGB
   Hovered Opacity=@n12.A
   Trans_Hovered=@n13
   IsPressed Active=@n14
   Pressed=@n15.RGB
   Pressed Opacity=@n15.A
   Trans_Pressed=@n16
   Is Disabled Active=@n17
   Disabled=@n18.RGB
   Disabled Opacity=@n18.A
   Trans_Disabled=@n19
  ) {materialFunction=/FNE_UI/Systems/_MaterialFunctions/ButtonStates/MF_FNEUI_ButtonStatesBase_Lerp}
  NamedRerouteDeclaration(None=@n20.Color) {name=Icon Colored by States, nodeColor=<rgba 1 0 0.4922 1>}
  n21 = Max(A=Max(A=@n6.R, B=@n6.R), B=@n6.B)
  n22 = TextureCoordinate()
  n23 = Multiply(
   A=Multiply(A=NamedRerouteDeclaration(None=@n20.Opacity) {name=State Opacities, nodeColor=<rgba 0.0547 0 1 1>}, B=@n21)
   B=NamedRerouteDeclaration(None=MaterialFunctionCall!(UVs=@n22, Alpha=@n6.A) {materialFunction=/FNE_UI/Systems/_MaterialFunctions/Basics/MF_FNEUI_ClipEdgePixel}.Result) {name=Icon Alpha, nodeColor=<rgba 0 0.64 1 1>}
  )
  NamedRerouteDeclaration(None=@n23) {name=Max White Alpha, nodeColor=<rgba 0 0.64 1 1>}
  n24 = ScalarParameter() {
   desc="Editing this parameter will force the setting and override the blueprint setting.\r\n"
   group="[0] Do Not Edit"
   parameterName=Force Black and White
   sliderMax=1
   sortPriority=1
  }
  NamedRerouteDeclaration(None=Saturate(None=Floor(None=@n24))) {name=Is Black and White, nodeColor=<rgba 0 0 0 1>}
  NamedRerouteDeclaration(None=@n21) {name=Max White RGB, nodeColor=<rgba 0.6172 1 0 1>}
  n25 = ScalarParameter() {
   desc="Editing this parameter will force the setting and override the blueprint setting.\r\n"
   group="[0] Do Not Edit"
   parameterName=Force Transparent Black
   sliderMax=1
   sortPriority=2
  }
  NamedRerouteDeclaration(None=Saturate(None=Floor(None=@n25))) {name=Is Black Transparent, nodeColor=<rgba 0 0 0 1>}
  NamedRerouteDeclaration(None=MaterialFunctionCall!(RangeMin=Constant() {desc="B = 0 is too aliased", r=0.1}, RangeMax=Constant() {r=1}, Value=@n23) {materialFunction=/Engine/Functions/Engine_MaterialFunctions03/Math/WithinRange}.AllComponentsWithinRange) {name=Pure White Alpha, nodeColor=<rgba 0 0.64 1 1>}
 ]
}
