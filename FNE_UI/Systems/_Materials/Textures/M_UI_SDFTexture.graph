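# Translucent UI material that renders a single-channel SDF texture, a heart icon by default, as a tinted, rotatable shape with strokes, glow and drop shadow.
# SmoothStep thresholds on the red channel produce fill, inner stroke and outer stroke masks with parameterized thickness and softness.
# The shadow resamples the texture with its own rotation and offset.
# Named reroutes feed these layers into lerps for final color and Max nodes for final alpha.
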Material /FNE_UI/Systems/_Materials/Textures/M_UI_SDFTexture
graph {
 properties {bAutomaticallySetUsageInEditor=false, blendMode=BLEND_Translucent, materialDomain=MD_UI, usedShadingModels=Default Lit}
 nodes [
  NamedRerouteDeclaration(None=ScalarParameter() {group="[3] Stroke", parameterName=InnerStrokeThickness, sliderMax=1}) {name=InnerStrokeThickness, nodeColor=<rgba 0 0 0 1>}
  NamedRerouteDeclaration(None=ScalarParameter() {defaultValue=0.05, group="[1] Base", parameterName=GlowMax, sliderMax=1, sliderMin=0.01}) {name=Glow, nodeColor=<rgba 0 0 0 1>}
  NamedRerouteDeclaration(None=ScalarParameter() {group="[2] Shadow", parameterName=ShadowOpacity, sliderMax=1}) {name=ShadowOpacity, nodeColor=<rgba 0 0 0 1>}
  NamedRerouteDeclaration(None=MaterialFunctionCall(UVs=TextureCoordinate(), "Rotation Angle (0-1)"=ScalarParameter() {group="[1] Base", parameterName=Rotation, sliderMax=1}) {materialFunction=&CustomRotator}.Rotated Values) {name=UV, nodeColor=<rgba 0 0.8594 1 1>}
  n1 = LinearInterpolate(
   A=VectorParameter() {group="[3] Stroke", parameterName=OuterStrokeColor, defaultValue=<rgba 1 1 1 1>}.RGB
   B=VectorParameter() {group="[3] Stroke", parameterName=InnerStrokeColor, defaultValue=<rgba 1 1 1 1>}.RGB
   Alpha=NamedRerouteUsage()
  )
  n2 = VectorParameter() {group="[1] Base", parameterName=Color, defaultValue=<rgba 1 1 1 1>}
  n3 = NamedRerouteUsage()
  n4 = NamedRerouteUsage()
  n5 = NamedRerouteUsage()
  NamedRerouteDeclaration(None=LinearInterpolate(A=NamedRerouteUsage(), B=LinearInterpolate(A=@n1, B=@n2.RGB, Alpha=@n3), Alpha=Max(A=@n4, B=@n5))) {name=FinalColor, nodeColor=<rgba 1 0 0.8672 1>}
  n6 = NamedRerouteUsage()
  n7 = Subtract(A=NamedRerouteUsage(), B=NamedRerouteUsage())
  n8 = Subtract(A=Subtract(A=@n7, B=Multiply(A=@n7, B=@n6)), B=Multiply(A=@n7, B=NamedRerouteUsage()))
  n9 = NamedRerouteUsage()
  n10 = Multiply(A=NamedRerouteUsage(), B=NamedRerouteUsage())
  n11 = NamedRerouteUsage()
  n12 = NamedRerouteUsage()
  NamedRerouteDeclaration(None=SmoothStep(Min=Clamp(None=@n8), Max=Subtract(A=@n9, B=@n10), Value=TextureSample(UVs=@n11, Tex=@n12) {samplerSource=SSM_Clamp_WorldGroupSettings}.R)) {name=OuterStrokeAlpha, nodeColor=<rgba 0 1 0.4141 1>}
  NamedRerouteDeclaration(None=ScalarParameter() {group="[3] Stroke", parameterName=OuterStrokeGlowMax, sliderMax=1}) {name=OuterStrokeGlow, nodeColor=<rgba 0 0 0 1>}
  NamedRerouteDeclaration(None=ScalarParameter() {group="[3] Stroke", parameterName=OuterStrokeThicknessGlowMax, sliderMax=1}) {name=OuterStrokeThicknessGlowMax, nodeColor=<rgba 0 0 0 1>}
  NamedRerouteDeclaration(None=ScalarParameter() {group="[2] Shadow", parameterName=ShadowThickness, sliderMax=1}) {name=ShadowThickness, nodeColor=<rgba 0 0 0 1>}
  NamedRerouteDeclaration(None=ScalarParameter() {group="[2] Shadow", parameterName=ShadowRotation, sliderMax=1}) {name=ShadowRotation, nodeColor=<rgba 0 0 0 1>}
  n13 = Subtract(A=NamedRerouteUsage(), B=NamedRerouteUsage())
  n14 = Add(A=@n13, B=Multiply(A=OneMinus(None=@n13), B=NamedRerouteUsage()))
  n15 = NamedRerouteUsage()
  n16 = OneMinus(None=@n15)
  n17 = NamedRerouteUsage()
  n18 = Add(A=Add(A=@n15, B=Multiply(A=@n16, B=@n17)), B=Multiply(A=OneMinus(None=@n15), B=NamedRerouteUsage()))
  n19 = NamedRerouteUsage()
  n20 = NamedRerouteUsage()
  NamedRerouteDeclaration(None=SmoothStep(Min=Clamp(None=@n14), Max=Clamp(None=@n18), Value=TextureSample(UVs=@n19, Tex=@n20) {samplerSource=SSM_Clamp_WorldGroupSettings}.R)) {name=InnerStrokeAlpha, nodeColor=<rgba 0 1 0.4141 1>}
  NamedRerouteDeclaration(None=ScalarParameter() {group="[3] Stroke", parameterName=OuterStrokeThickness, sliderMax=1}) {name=OuterStrokeThickness, nodeColor=<rgba 0 0 0 1>}
  NamedRerouteDeclaration(None=ScalarParameter() {defaultValue=0.16, group="[1] Base", parameterName=TextureFeatherAmount, sliderMax=1}) {name=SDFTextureStep, nodeColor=<rgba 0 0 0 1>}
  NamedRerouteDeclaration(None=ScalarParameter() {group="[3] Stroke", parameterName=InnerStrokeThicknessGlowMax, sliderMax=1}) {name=InnerStrokeThicknessGlowMax, nodeColor=<rgba 0 0 0 1>}
  n21 = Subtract(A=NamedRerouteUsage(), B=NamedRerouteUsage())
  n22 = NamedRerouteUsage()
  n23 = NamedRerouteUsage()
  NamedRerouteDeclaration(None=SmoothStep(Min=Clamp(None=@n21), Max=NamedRerouteUsage(), Value=TextureSample(UVs=@n22, Tex=@n23).R)) {name=TextureAlpha, nodeColor=<rgba 0 1 0.0859 1>}
  n24 = NamedRerouteUsage()
  n25 = NamedRerouteUsage()
  NamedRerouteDeclaration(None=Max(A=Max(A=@n24, B=@n25), B=NamedRerouteUsage())) {name=FinalAlpha, nodeColor=<rgba 1 0 0.2344 1>}
  NamedRerouteDeclaration(None=ScalarParameter() {group="[3] Stroke", parameterName=InnerStrokeGlowMax, sliderMax=1}) {name=InnerStrokeGlow, nodeColor=<rgba 0 0 0 1>}
  NamedRerouteDeclaration(None=TextureObjectParameter() {group="[1] Base", parameterName=Texture, sortPriority=0, texture=/FNE_UI/Systems/_Textures/T_UI_Icon_SDF_heart_16px}) {name=Texture, nodeColor=<rgba 0 0.1328 1 1>}
  NamedRerouteDeclaration(None=ScalarParameter() {group="[2] Shadow", parameterName=ShadowOffsetX, sliderMax=1, sliderMin=-1}) {name=ShadowOffsetX, nodeColor=<rgba 0 0 0 1>}
  NamedRerouteDeclaration(None=ScalarParameter() {group="[2] Shadow", parameterName=ShadowOffsetY, sliderMax=1, sliderMin=-1}) {name=ShadowOffsetY, nodeColor=<rgba 0 0 0 1>}
  NamedRerouteDeclaration(None=ScalarParameter() {group="[2] Shadow", parameterName=ShadowGlowMax, sliderMax=1}) {name=ShadowGlow, nodeColor=<rgba 0 0 0 1>}
  n26 = Subtract(A=NamedRerouteUsage(), B=Multiply(A=NamedRerouteUsage(), B=NamedRerouteUsage()))
  n27 = MaterialFunctionCall(UVs=TextureCoordinate(), "Rotation Angle (0-1)"=NamedRerouteUsage()) {materialFunction=&CustomRotator}
  n28 = MaterialFunctionCall(X=NamedRerouteUsage(), Y=NamedRerouteUsage()) {materialFunction=&MakeFloat2}
  n29 = SmoothStep(
   Min=@n26
   Max=Add(A=@n26, B=NamedRerouteUsage())
   Value=TextureSample(
    UVs=MaterialFunctionCall!(Input UVs=@n27.Rotated Values, Translate=@n28.Result) {materialFunction=/FNE_UI/Systems/_MaterialFunctions/Transforms/MF_UI_Translate}.UVs raw
    Tex=NamedRerouteUsage()
   ) {samplerSource=SSM_Clamp_WorldGroupSettings}.R
  )
  NamedRerouteDeclaration(None=Multiply(A=@n29, B=VectorParameter() {group="[2] Shadow", parameterName=ShadowColor, defaultValue=<rgba 0 0 0 1>}.RGB)) {name=ShadowRGB, nodeColor=<rgba 0.7344 0 1 1>}
  NamedRerouteDeclaration(None=Multiply(A=@n29, B=NamedRerouteUsage())) {name=ShadowAlpha, nodeColor=<rgba 0 1 0.4141 1>}
  NamedRerouteUsage()
  NamedRerouteUsage()
 ]
}
