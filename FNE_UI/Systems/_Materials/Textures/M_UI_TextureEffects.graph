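# Translucent UI material that applies optional animated effects and color treatments to a texture parameter, built mostly from shared FNE_UI material functions.
# UV effects cover panning, grid tiling, ping-pong scaling with per-cell random variation, wave warping, pixelation and sine bouncing.
# Color effects cover halftone dots, stepped gradients, and three-stop linear or radial gradients for tint and alpha color.
# Scalar parameters blend each effect through lerps, with named reroutes connecting the stages.
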Material /FNE_UI/Systems/_Materials/Textures/M_UI_TextureEffects
graph {
 properties {bAutomaticallySetUsageInEditor=false, blendMode=BLEND_Translucent, materialDomain=MD_UI, usedShadingModels=Default Lit}
 nodes [
  n1 = NamedRerouteUsage()
  n2 = ScalarParameter() {defaultValue=0.01, group="[3] Warping", parameterName=AmplitudeX}
  n3 = ScalarParameter() {defaultValue=0.02, group="[3] Warping", parameterName=AmplitudeY}
  n4 = ScalarParameter() {group="[3] Warping", parameterName=FrequencyX}
  n5 = ScalarParameter() {defaultValue=0.5, group="[3] Warping", parameterName=FrequencyY}
  n6 = ScalarParameter() {defaultValue=0.15, desc="How fast the texture is warping \r\n(only if IsWarping = 1)", group="[3] Warping", parameterName=WarpSpeed}
  n7 = MaterialFunctionCall!(FloatValue, Speed=Multiply(A=Constant() {r=1}, B=@n6)) {materialFunction=/FNE_UI/Systems/_MaterialFunctions/Time/MF_FNEUI_LinearTime}
  n8 = ScalarParameter() {
   desc="Determines if the texture warps as if it's underwater.\r\n0.0 = No warping\r\n1.0 = Warping"
   group="[3] Warping"
   parameterName=IsWarping
   sliderMax=1
   sortPriority=0
  }
  NamedRerouteDeclaration(None=LinearInterpolate(
    A=NamedRerouteUsage()
    B=MaterialFunctionCall!(UVs=@n1, Amplitude X=@n2, Amplitude Y=@n3, Frequency X=@n4, Frequency Y=@n5, Phase, Motion=@n7.0-1 Loop) {materialFunction=/FNE_UI/Systems/_MaterialFunctions/Distortions/MF_UI_WaveWarp}.Warped UVs
    Alpha=@n8
   )) {name=WarpingUV, nodeColor=<rgba 0 1 0.7188 1>}
  n9 = TextureCoordinate()
  n10 = AppendVector(A=ScalarParameter() {group="[1] Basics", parameterName=TranslateX}, B=ScalarParameter() {group="[1] Basics", parameterName=TranslateY})
  n11 = ScalarParameter() {group="[1] Basics", parameterName=TranslateSpeedX, sortPriority=33}
  n12 = ScalarParameter() {group="[1] Basics", parameterName=TranslateSpeedY, sortPriority=33}
  NamedRerouteDeclaration(None=Panner(
    Coordinate=MaterialFunctionCall!(Input UVs=@n9, Translate=@n10) {materialFunction=/FNE_UI/Systems/_MaterialFunctions/Transforms/MF_UI_Translate}.UVs raw
    Speed=AppendVector(A=@n11, B=@n12)
   )) {name=UVs, nodeColor=<rgba 0 0.9531 1 1>}
  n13 = VectorParameter() {group="[9] Alpha Color", parameterName=AlphaColor1, defaultValue=<rgba 1 1 1 1>}
  NamedRerouteDeclaration(None=@n13.RGB) {name=AlphaColor, nodeColor=<rgba 0.4297 0 1 1>}
  n14 = Floor(None=Multiply(A=TextureCoordinate(), B=NamedRerouteUsage()))
  n15 = Constant2Vector() {r=1}
  NamedRerouteDeclaration(None=MaterialFunctionCall!(Float 2=Multiply(A=@n14, B=@n15) {constB=0}, Seed=Constant() {r=1}, Single value) {materialFunction=/FNE_UI/Systems/_MaterialFunctions/Utilities/MF_UI_Random}.1D) {name=VerticalNoise, nodeColor=<rgba 0 1 0.6016 1>}
  n16 = NamedRerouteUsage()
  n17 = LinearInterpolate(Alpha=ScalarParameter() {group="[7] Stepped Gradient", parameterName=SteppedGradientAmount, sliderMax=1}) {constA=-0.2, constB=1.2}
  n18 = MaterialFunctionCall!(Gradient, Num steps=LinearInterpolate(Alpha=Saturate(None=@n17)) {constA=2, constB=8}) {materialFunction=/FNE_UI/Systems/_MaterialFunctions/Utilities/MF_UI_SteppedGradient}
  n19 = ScalarParameter() {
   desc="A gradient with vertical or horizontal segments or steps, like steps."
   group="[7] Stepped Gradient"
   parameterName=IsSteppedGradient
   sliderMax=1
   sortPriority=0
  }
  n20 = ScalarParameter() {defaultValue=-0.01, group="[6] Halftone", parameterName=HalftonePanSpeedX}
  n21 = ScalarParameter() {defaultValue=0.01, group="[6] Halftone", parameterName=HalftonePanSpeedY}
  NamedRerouteDeclaration(None=MaterialFunctionCall!(
    UVs
    Input gradient=LinearInterpolate(A=@n16, B=@n18.Output keep initial range, Alpha=@n19)
    Dot density=ScalarParameter() {defaultValue=30, group="[6] Halftone", parameterName=DotDensity}
    Dot min size=ScalarParameter() {defaultValue=-0.2, group="[6] Halftone", parameterName=DotMinSize}
    Dot max size=ScalarParameter() {defaultValue=1, group="[6] Halftone", parameterName=DotMaxSize}
    Dot glow min=ScalarParameter() {group="[6] Halftone", parameterName=DotGlowMin}
    Dot glow max=ScalarParameter() {defaultValue=0.5, group="[6] Halftone", parameterName=DotGlowMax}
    Halftone pan speed=AppendVector(A=@n20, B=@n21)
    Halftone pattern rotation=ScalarParameter() {group="[6] Halftone", parameterName=HalftoneRotation}
    Container ratio
   ) {materialFunction=/FNE_UI/Systems/_MaterialFunctions/Patterns/MF_UI_Halftone}.Dot halftone) {name=HalftoneDotsUV, nodeColor=<rgba 0 0.3906 1 1>}
  n22 = ScalarParameter() {
   defaultValue=0.1
   desc="Lower range of how fast the texture scales \r\n\r\n(only if IsScaling = 1)"
   group="[2] Scaling"
   parameterName=ScalingMinimumSpeed
  }
  n23 = ScalarParameter() {
   defaultValue=0.3
   desc="Upper range of how quick the texture scales, will ping pong back to the minimum speed\r\n\r\n(only if IsScaling = 1)"
   group="[2] Scaling"
   parameterName=ScalingMaximumSpeed
  }
  n24 = LinearInterpolate(A=@n22, B=@n23, Alpha=NamedRerouteUsage()) {constA=0.5}
  NamedRerouteDeclaration(None=MaterialFunctionCall(Speed=@n24) {materialFunction=&MF_FNEUI_LinearTime}.0-1 Ping pong) {name=Lifetime, nodeColor=<rgba 1 0.6562 0 1>}
  n25 = TextureObjectParameter() {
   group="[1] Basics"
   parameterName=Texture
   texture=/Game/Athena/UI/InGame/Creative/PropertyEditor/Icon_Picker/Textures_V2/128x128/T_UI_IconLibrary_Llama_128
  }
  n26 = TextureSample(UVs=NamedRerouteUsage(), Tex=@n25) {samplerSource=SSM_Clamp_WorldGroupSettings}
  NamedRerouteDeclaration(None=@n26.R) {name=TextureRedChannel, nodeColor=<rgba 1 0 0.0511 1>}
  NamedRerouteDeclaration(None=@n26.A) {name=TextureAlpha, nodeColor=<rgba 1 0.1875 0 1>}
  NamedRerouteDeclaration(None=@n26.RGB) {name=TextureRGB, nodeColor=<rgba 0 0.1328 1 1>}
  NamedRerouteDeclaration(None=MaterialFunctionCall!(UV=NamedRerouteUsage(), Tiling X=NamedRerouteUsage(), Tiling Y=NamedRerouteUsage()) {materialFunction=/FNE_UI/Systems/_MaterialFunctions/Patterns/MF_UI_GridTiling}.ScaledPatternUVs) {name=GridTilingUV, nodeColor=<rgba 1 0 0.7031 1>}
  n27 = VectorParameter() {desc=First color in the tint gradient., group="[8] Tint Color", parameterName=TintColor1, defaultValue=<rgba 1 1 1 1>}
  NamedRerouteDeclaration(None=@n27.RGB) {name=Color, nodeColor=<rgba 0 1 0.1094 1>}
  n28 = TextureCoordinate()
  n29 = ScalarParameter() {group="[9] Alpha Color", parameterName=AlphaGradientRotation, sliderMax=1}
  n30 = LinearInterpolate(
   A=MaterialFunctionCall(UV Channel=MaterialFunctionCall(UVs=@n28, "Rotation Angle (0-1)"=@n29) {materialFunction=&CustomRotator}.Rotated Values) {materialFunction=&LinearGradient}.UGradient
   B=MaterialFunctionCall!(
    UV
    Size
    GlowMin=ScalarParameter() {group="[9] Alpha Color", parameterName=AlphaGradientRadialSize, sortPriority=33}
    GlowMax=Constant() {r=1}
    Stroke thickness
    Is stroke thickness relative to size
   ) {materialFunction=/FNE_UI/Systems/_MaterialFunctions/SDFs/MF_UI_SDF_Circle}.Fill
   Alpha=ScalarParameter() {group="[9] Alpha Color", parameterName=AlphaRadialGradient, sliderMax=1}
  )
  n31 = VectorParameter() {group="[9] Alpha Color", parameterName=AlphaColor2, defaultValue=<rgba 1 1 1 1>}
  n32 = VectorParameter() {group="[9] Alpha Color", parameterName=AlphaColor3, defaultValue=<rgba 1 1 1 1>}
  n33 = ScalarParameter() {group="[9] Alpha Color", parameterName=AlphaColor1Position, sliderMax=1}
  n34 = ScalarParameter() {defaultValue=0.5, group="[9] Alpha Color", parameterName=AlphaColor2Position, sliderMax=1}
  n35 = ScalarParameter() {defaultValue=1, group="[9] Alpha Color", parameterName=AlphaColor3Position, sliderMax=1}
  NamedRerouteDeclaration(None=LinearInterpolate(
    A=NamedRerouteUsage()
    B=MaterialFunctionCall!(
     In=@n30
     Value 1=@n13.RGB
     Value 2=@n31.RGB
     Value 3=@n32.RGB
     Value 4=@n32.RGB
     Value 5=@n32.RGB
     Pos value 1=@n33
     Pos value 2=@n34
     Pos value 3=@n35
     Pos value 4=@n35
     Pos value 5=@n35
    ) {materialFunction=/FNE_UI/Systems/_MaterialFunctions/Utilities/MF_UI_RemappableGradient}.Gradient 3 Colors
    Alpha=ScalarParameter() {defaultValue=1, group="[9] Alpha Color", parameterName=AlphaGradientOn, sliderMax=1}
   )) {name=AlphaGradient, nodeColor=<rgba 0.5 1 0 1>}
  n36 = ScalarParameter() {
   desc="Determines if the texture will scale up and down repeatedly\r\n0.0 = No scaling\r\n1.0 = Scaling"
   group="[2] Scaling"
   parameterName=IsScaling
   sliderMax=1
   sortPriority=0
  }
  n37 = ScalarParameter() {
   desc="Determines if the texture warps as if it's underwater.\r\n0.0 = No warping\r\n1.0 = Warping"
   group="[3] Warping"
   parameterName=IsWarping
   sliderMax=1
   sortPriority=0
  }
  n38 = LinearInterpolate(A=LinearInterpolate(A=NamedRerouteUsage(), B=NamedRerouteUsage(), Alpha=@n36), B=NamedRerouteUsage(), Alpha=@n37)
  n39 = NamedRerouteUsage()
  n40 = ScalarParameter() {desc=Pixelates the texture., group="[4] Pixelate", parameterName=IsPixelated, sliderMax=1, sortPriority=0}
  NamedRerouteDeclaration(None=LinearInterpolate(
    A=LinearInterpolate(A=@n38, B=@n39, Alpha=@n40)
    B=NamedRerouteUsage()
    Alpha=ScalarParameter() {desc=Bounces the texture., group="[5] Bounce", parameterName=IsBouncing, sliderMax=1, sortPriority=0}
   )) {name=FinalUVs, nodeColor=<rgba 0 1 0.0859 1>}
  n41 = TextureCoordinate()
  n42 = ScalarParameter() {desc=Rotation of the tint gradient., group="[8] Tint Color", parameterName=TintGradientRotation, sliderMax=1}
  n43 = ScalarParameter() {desc=Sets the size of the radial gradient for the tint., group="[8] Tint Color", parameterName=TintGradientRadialSize, sortPriority=33}
  n44 = ScalarParameter() {
   desc="Sets the gradient type for the tint.\r\n0.0 = Linear Gradient\r\n1.0 = Radial Gradient"
   group="[8] Tint Color"
   parameterName=TintRadialGradient
   sliderMax=1
  }
  n45 = LinearInterpolate(
   A=MaterialFunctionCall(UV Channel=MaterialFunctionCall(UVs=@n41, "Rotation Angle (0-1)"=@n42) {materialFunction=&CustomRotator}.Rotated Values) {materialFunction=&LinearGradient}.UGradient
   B=MaterialFunctionCall(GlowMin=@n43, GlowMax=Constant() {r=1}) {materialFunction=&MF_UI_SDF_Circle}.Fill
   Alpha=@n44
  )
  n46 = VectorParameter() {desc=Second color in the tint gradient., group="[8] Tint Color", parameterName=TintColor2, defaultValue=<rgba 1 1 1 1>}
  n47 = VectorParameter() {desc=Third color in the tint gradient., group="[8] Tint Color", parameterName=TintColor3, defaultValue=<rgba 1 1 1 1>}
  n48 = ScalarParameter() {desc=Sets the position of the first color in the tint gradient., group="[8] Tint Color", parameterName=TintColor1Position, sliderMax=1}
  n49 = ScalarParameter() {
   defaultValue=0.5
   desc=Sets the position of the second color in the tint gradient.
   group="[8] Tint Color"
   parameterName=TintColor2Position
   sliderMax=1
  }
  n50 = ScalarParameter() {
   defaultValue=1
   desc=Sets the position of the third color in the tint gradient.
   group="[8] Tint Color"
   parameterName=TintColor3Position
   sliderMax=1
  }
  n51 = ScalarParameter() {
   defaultValue=1
   desc="Sets to use gradient or first color\r\n0.0 = tint texture with TintColor1\r\n1.0 = tint texture with linear gradient"
   group="[8] Tint Color"
   parameterName=TintGradientOn
   sliderMax=1
  }
  NamedRerouteDeclaration(None=LinearInterpolate(
    A=NamedRerouteUsage()
    B=MaterialFunctionCall(
     In=@n45
     Value 1=@n27.RGB
     Value 2=@n46.RGB
     Value 3=@n47.RGB
     Value 4=@n47.RGB
     Value 5=@n47.RGB
     Pos value 1=@n48
     Pos value 2=@n49
     Pos value 3=@n50
     Pos value 4=@n50
     Pos value 5=@n50
    ) {materialFunction=&MF_UI_RemappableGradient}.Gradient 3 Colors
    Alpha=@n51
   )) {name=Gradient, nodeColor=<rgba 1 0 0.8203 1>}
  n52 = NamedRerouteUsage()
  n53 = LinearInterpolate(Alpha=ScalarParameter() {group="[7] Stepped Gradient", parameterName=SteppedGradientAmount, sliderMax=1}) {constA=-0.2, constB=1.2}
  n54 = MaterialFunctionCall(Gradient=NamedRerouteUsage(), Num steps=LinearInterpolate(Alpha=Saturate(None=@n53)) {constA=2, constB=8}) {materialFunction=&MF_UI_SteppedGradient}
  n55 = ScalarParameter() {
   desc="A gradient with vertical or horizontal segments or steps, like steps."
   group="[7] Stepped Gradient"
   parameterName=IsSteppedGradient
   sliderMax=1
   sortPriority=0
  }
  n56 = VectorParameter() {group="[6] Halftone", parameterName=GradientColor1, defaultValue=<rgba 1 0.0608 0 1>}
  n57 = VectorParameter() {group="[6] Halftone", parameterName=GradientColor2, defaultValue=<rgba 1 0.9365 0 1>}
  n58 = NamedRerouteUsage()
  NamedRerouteDeclaration(None=LinearInterpolate(
    A=LinearInterpolate(A=@n52, B=@n54.Output keep initial range, Alpha=@n55)
    B=LinearInterpolate(A=@n56.RGB, B=@n57.RGB, Alpha=@n58)
    Alpha=ScalarParameter() {group="[6] Halftone", parameterName=IsHalftone, sliderMax=1, sortPriority=0}
   )) {name=FinalRGB, nodeColor=<rgba 0 1 0.3203 1>}
  n59 = Constant()
  n60 = NamedRerouteUsage()
  n61 = LinearInterpolate(
   A=Divide(A=@n60) {constB=-2}
   B=Divide(A=@n60)
   Alpha=MaterialFunctionCall(Float 2=@n24) {materialFunction=&MF_UI_Random}.1D
  )
  n62 = ScalarParameter() {defaultValue=0.2, desc="Starting scale for scaling textures \r\n\r\n(only if IsScaling = 1)", group="[2] Scaling", parameterName=StartSize}
  n63 = ScalarParameter() {
   defaultValue=1
   desc="In between scale for the texture \r\n\r\n(only if IsScaling = 1)"
   group="[2] Scaling"
   parameterName=MiddleSize
   sortPriority=33
  }
  n64 = ScalarParameter() {
   desc="End scale of the texture, which will then ping pong back to the StartSize \r\n\r\n(only if IsScaling = 1)"
   group="[2] Scaling"
   parameterName=EndSize
   sortPriority=34
  }
  n65 = MaterialFunctionCall!(
   UVs=MaterialFunctionCall(Input UVs=NamedRerouteUsage(), Translate=AppendVector(A=@n59, B=@n61)) {materialFunction=&MF_UI_Translate}.UVs raw
   Scale=MaterialFunctionCall!(1=@n62, 2=@n63, 3=@n64, 4, A=NamedRerouteUsage()) {materialFunction=/Engine/Functions/Engine_MaterialFunctions02/Lerp_Multiple_Float}.Lerp 3 Inputs
   Scale center
  ) {materialFunction=/FNE_UI/Systems/_MaterialFunctions/Transforms/MF_UI_Scale}
  n66 = NamedRerouteUsage()
  n67 = ScalarParameter() {
   desc="Determines if the texture will scale up and down repeatedly\r\n0.0 = No scaling\r\n1.0 = Scaling"
   group="[2] Scaling"
   parameterName=IsScaling
   sliderMax=1
   sortPriority=0
  }
  NamedRerouteDeclaration(None=LinearInterpolate(
    A=NamedRerouteUsage()
    B=MaterialFunctionCall(UVs=@n65.Result, "Rotation Angle (0-1)"=@n66) {materialFunction=&CustomRotator}.Rotated Values
    Alpha=@n67
   )) {name=ScalingUV, nodeColor=<rgba 0.5703 0 1 1>}
  NamedRerouteDeclaration(None=ScalarParameter() {defaultValue=1, desc="Number of horizontal segments \r\n(only if IsScaling = 1)", group="[1] Basics", parameterName=TilingX}) {name=TilingX, nodeColor=<rgba 1 0.8203 0 1>}
  NamedRerouteDeclaration(None=ScalarParameter() {defaultValue=1, desc="Number of vertical segments \r\n(only if IsScaling = 1)", group="[1] Basics", parameterName=TilingY}) {name=TilingY, nodeColor=<rgba 0.5 0 1 1>}
  n68 = NamedRerouteUsage()
  n69 = ScalarParameter() {defaultValue=40, group="[4] Pixelate", parameterName=NumberOfPixels}
  NamedRerouteDeclaration(None=LinearInterpolate(
    A=@n68
    B=MaterialFunctionCall!(UVs=@n68, NumPixels=@n69) {materialFunction=/FNE_UI/Systems/_MaterialFunctions/Utilities/MF_UI_Pixelate}.Result
    Alpha=ScalarParameter() {desc=Pixelates the texture., group="[4] Pixelate", parameterName=IsPixelated, sliderMax=1, sortPriority=0}
   )) {name=PixelateUV, nodeColor=<rgba 0 0.5312 1 1>}
  n70 = NamedRerouteUsage()
  n71 = AppendVector(
   A=ScalarParameter() {defaultValue=0.05, group="[5] Bounce", parameterName=BounceX}
   B=ScalarParameter() {defaultValue=0.05, group="[5] Bounce", parameterName=BounceY}
  )
  n72 = LinearInterpolate(
   A=@n71
   B=Multiply(A=@n71) {constB=-1}
   Alpha=MaterialFunctionCall!(Speed=ScalarParameter() {defaultValue=1, group="[5] Bounce", parameterName=BounceSpeed}) {materialFunction=/FNE_UI/Systems/_MaterialFunctions/Time/MF_UI_SineTime}.0-1 sine time
  )
  NamedRerouteDeclaration(None=LinearInterpolate(
    A=@n70
    B=MaterialFunctionCall(Input UVs=@n70, Translate=@n72) {materialFunction=&MF_UI_Translate}.UVs raw
    Alpha=ScalarParameter() {desc=Bounces the texture., group="[5] Bounce", parameterName=IsBouncing, sliderMax=1, sortPriority=0}
   )) {name=BounceUV, nodeColor=<rgba 0.5 0 1 1>}
  n73 = NamedRerouteUsage()
  n74 = NamedRerouteUsage()
  n75 = ScalarParameter() {group="[9] Alpha Color", parameterName=AlphaOpacity, sliderMax=1, sortPriority=0}
  n76 = NamedRerouteUsage()
  n77 = NamedRerouteUsage()
  n78 = ScalarParameter() {desc=Sets the transparency of the tint gradient., group="[8] Tint Color", parameterName=TintGradientOpacity, sliderMax=1, sortPriority=0}
  n79 = NamedRerouteUsage()
  n80 = NamedRerouteUsage()
  NamedRerouteDeclaration(None=LinearInterpolate(A=LinearInterpolate(A=@n73, B=@n74, Alpha=@n75), B=LinearInterpolate(A=@n76, B=@n77, Alpha=@n78), Alpha=Multiply(A=@n79, B=@n80))) {name=TextureTintColor, nodeColor=<rgba 1 0 0.75 1>}
  n81 = NamedRerouteUsage()
  n82 = NamedRerouteUsage()
  NamedRerouteDeclaration(None=LinearInterpolate(A=@n81, B=Min(A=@n81, B=@n82), Alpha=ScalarParameter() {group="[6] Halftone", parameterName=IsHalftone, sliderMax=1, sortPriority=0})) {name=HalftoneUV, nodeColor=<rgba 0 1 0.25 1>}
  NamedRerouteUsage()
  NamedRerouteUsage()
 ]
}
