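# UI material that draws a positionable, scalable icon texture over a shape with gradient fill and outline, plus a texture mask.
# A Shape parameter switches between box, circle, hexagon and triangle SDF functions on rotated, slanted UVs, separating fill, outline and gap.
# Three-color gradients tint fill (radial or linear) and outline.
# A movable box mask combines with these layers through min, max and lerp into final color and alpha, with debug preview.
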Material /FNE_UI/Systems/_Materials/Textures/M_UI_TextureMask
graph {
 expressions {
  MaterialExpressionAppendVector_0 {
   class=MaterialExpressionAppendVector
   inputs [
    A {from=MaterialExpressionScalarParameter_3, output=""}
    B {from=MaterialExpressionScalarParameter_4, output=""}
   ]
  }
  MaterialExpressionAppendVector_1 {
   class=MaterialExpressionAppendVector
   inputs [
    A {from=MaterialExpressionScalarParameter_2, output=""}
    B {from=MaterialExpressionScalarParameter_5, output=""}
   ]
  }
  MaterialExpressionAppendVector_2 {
   class=MaterialExpressionAppendVector
   inputs [
    A {from=MaterialExpressionScalarParameter_34, output=""}
    B {from=MaterialExpressionScalarParameter_33, output=""}
   ]
  }
  MaterialExpressionAppendVector_3 {
   class=MaterialExpressionAppendVector
   inputs [
    A {from=MaterialExpressionScalarParameter_35, output=""}
    B {from=MaterialExpressionScalarParameter_36, output=""}
   ]
  }
  MaterialExpressionAppendVector_4 {
   class=MaterialExpressionAppendVector
   inputs [
    A {from=MaterialExpressionScalarParameter_13, output=""}
    B {from=MaterialExpressionScalarParameter_18, output=""}
   ]
  }
  MaterialExpressionConstant2Vector_0 {class=MaterialExpressionConstant2Vector, properties {g=0.5, r=0.5}}
  MaterialExpressionConstant3Vector_0 {class=MaterialExpressionConstant3Vector, properties {constant=<rgba 1 0 0 1>}}
  MaterialExpressionConstant4Vector_0 {class=MaterialExpressionConstant4Vector, inputs [], properties {constant=<rgba 0.1 0.1 0.1 0.1>}}
  MaterialExpressionConstant_0 {class=MaterialExpressionConstant, properties {r=0.5}}
  MaterialExpressionConstant_2 {class=MaterialExpressionConstant}
  MaterialExpressionDivide_1 {
   class=MaterialExpressionDivide
   inputs [
    A {from=MaterialExpressionScalarParameter_0}
    B {from=""}
   ]
   properties {constA=360, constB=360}
  }
  MaterialExpressionLinearInterpolate_1 {
   class=MaterialExpressionLinearInterpolate
   inputs [
    A {from=""}
    B {from=MaterialExpressionMaterialFunctionCall_2, output=Gradient 3 Colors}
    Alpha {from=MaterialExpressionScalarParameter_16}
   ]
  }
  MaterialExpressionLinearInterpolate_17 {
   class=MaterialExpressionLinearInterpolate
   inputs [
    A {from=MaterialExpressionMax_16}
    B {from=MaterialExpressionMax_17}
    Alpha {from=MaterialExpressionScalarParameter_41}
   ]
  }
  MaterialExpressionLinearInterpolate_18 {
   class=MaterialExpressionLinearInterpolate
   inputs [
    A {from=MaterialExpressionNamedRerouteUsage_63}
    B {from=MaterialExpressionNamedRerouteUsage_64}
    Alpha {from=MaterialExpressionMin_7}
   ]
  }
  MaterialExpressionLinearInterpolate_19 {
   class=MaterialExpressionLinearInterpolate
   inputs [
    A {from=MaterialExpressionLinearInterpolate_18}
    B {from=MaterialExpressionMultiply_6}
    Alpha {from=MaterialExpressionNamedRerouteUsage_75}
   ]
  }
  MaterialExpressionLinearInterpolate_2 {
   class=MaterialExpressionLinearInterpolate
   inputs [
    A {from=""}
    B {from=MaterialExpressionMaterialFunctionCall_8, output=Gradient 3 Colors}
    Alpha {from=MaterialExpressionScalarParameter_17}
   ]
  }
  MaterialExpressionLinearInterpolate_20 {
   class=MaterialExpressionLinearInterpolate
   inputs [
    A {from=MaterialExpressionLinearInterpolate_19}
    B {from=MaterialExpressionNamedRerouteUsage_69}
    Alpha {from=MaterialExpressionMin_8}
   ]
  }
  MaterialExpressionLinearInterpolate_21 {
   class=MaterialExpressionLinearInterpolate
   inputs [
    A {from=MaterialExpressionLinearInterpolate_20}
    B {from=MaterialExpressionMax_18}
    Alpha {from=MaterialExpressionScalarParameter_42}
   ]
  }
  MaterialExpressionLinearInterpolate_4 {
   class=MaterialExpressionLinearInterpolate
   inputs [
    A {from=MaterialExpressionReroute_9}
    B {from=MaterialExpressionSubtract_14}
    Alpha {from=MaterialExpressionScalarParameter_20}
   ]
  }
  MaterialExpressionLinearInterpolate_9 {
   class=MaterialExpressionLinearInterpolate
   inputs [
    A {from=MaterialExpressionMaterialFunctionCall_6, output=RadialGradientExponential}
    B {from=MaterialExpressionMaterialFunctionCall_3, output=UGradient}
    Alpha {from=MaterialExpressionScalarParameter_23}
   ]
  }
  MaterialExpressionMaterialFunctionCall_0 {
   class=MaterialExpressionMaterialFunctionCall
   inputs [
    R {from=MaterialExpressionScalarParameter_9}
    G {from=MaterialExpressionScalarParameter_10}
    B {from=MaterialExpressionScalarParameter_12}
    A {from=MaterialExpressionScalarParameter_11}
   ]
   properties {materialFunction=&AppendMany}
  }
  MaterialExpressionMaterialFunctionCall_1 {
   class=MaterialExpressionMaterialFunctionCall
   inputs [
    UVs {from=MaterialExpressionTextureCoordinate_0}
    Rotation Center {from=MaterialExpressionConstant2Vector_0}
    "Rotation Angle (0-1)" {from=MaterialExpressionNamedRerouteUsage_0}
   ]
   properties {materialFunction=&CustomRotator}
  }
  MaterialExpressionMaterialFunctionCall_10 {
   class=MaterialExpressionMaterialFunctionCall
   inputs [
    UVs {from=MaterialExpressionReroute_4}
    Size {from=MaterialExpressionReroute_0}
    Outline Thickness {from=MaterialExpressionReroute_5}
    Glow Min {from=""}
    Glow Max {from=MaterialExpressionReroute_3}
    Corner Roundness {from=MaterialExpressionReroute_2}
    Keep Aspect Ratio {from=""}
   ]
   properties {materialFunction=/FNE_UI/Systems/_MaterialFunctions/Shapes/MF_FNEUI_Box}
  }
  MaterialExpressionMaterialFunctionCall_11 {
   class=MaterialExpressionMaterialFunctionCall
   inputs [
    UVs {from=MaterialExpressionMaterialFunctionCall_9, output=UVs raw}
    Size {from=MaterialExpressionAppendVector_3}
    Outline Thickness {from=""}
    Glow Min {from=""}
    Glow Max {from=MaterialExpressionConstant_2}
    Corner Roundness {from=MaterialExpressionConstant4Vector_0, output=RGBA}
    Keep Aspect Ratio {from=""}
   ]
   properties {materialFunction=&MF_FNEUI_Box}
  }
  MaterialExpressionMaterialFunctionCall_12 {
   class=MaterialExpressionMaterialFunctionCall
   inputs [
    UV {from=MaterialExpressionReroute_4}
    Size {from=MaterialExpressionReroute_0}
    GlowMin {from=""}
    GlowMax {from=MaterialExpressionReroute_3}
    Stroke thickness {from=MaterialExpressionReroute_5}
    Is stroke thickness relative to size {from=""}
   ]
   properties {materialFunction=/FNE_UI/Systems/_MaterialFunctions/SDFs/MF_UI_SDF_Circle}
  }
  MaterialExpressionMaterialFunctionCall_13 {
   class=MaterialExpressionMaterialFunctionCall
   inputs [
    Corner roundness {from=MaterialExpressionReroute_2}
    UVs {from=MaterialExpressionReroute_4}
    Radius {from=MaterialExpressionReroute_0}
    Glow Min {from=""}
    Glow Max {from=MaterialExpressionReroute_3}
    Outline Thickness {from=MaterialExpressionReroute_5}
    Is stroke thickness relative to size {from=""}
   ]
   properties {materialFunction=/FNE_UI/Systems/_MaterialFunctions/SDFs/MF_UI_SDF_Hexagon}
  }
  MaterialExpressionMaterialFunctionCall_14 {
   class=MaterialExpressionMaterialFunctionCall
   inputs [
    UVs {from=MaterialExpressionReroute_1}
    Size {from=MaterialExpressionReroute_6}
    Outline Thickness {from=""}
    Glow Min {from=""}
    Glow Max {from=""}
    Corner Roundness {from=MaterialExpressionReroute_7}
    Keep Aspect Ratio {from=""}
   ]
   properties {materialFunction=&MF_FNEUI_Box}
  }
  MaterialExpressionMaterialFunctionCall_15 {
   class=MaterialExpressionMaterialFunctionCall
   inputs [
    UV {from=MaterialExpressionReroute_1}
    Size {from=MaterialExpressionReroute_6}
    GlowMin {from=""}
    GlowMax {from=""}
    Stroke thickness {from=""}
    Is stroke thickness relative to size {from=""}
   ]
   properties {materialFunction=&MF_UI_SDF_Circle}
  }
  MaterialExpressionMaterialFunctionCall_16 {
   class=MaterialExpressionMaterialFunctionCall
   inputs [
    Corner roundness {from=MaterialExpressionReroute_7}
    UVs {from=MaterialExpressionReroute_1}
    Radius {from=MaterialExpressionReroute_6}
    Glow Min {from=""}
    Glow Max {from=""}
    Outline Thickness {from=""}
    Is stroke thickness relative to size {from=""}
   ]
   properties {materialFunction=&MF_UI_SDF_Hexagon}
  }
  MaterialExpressionMaterialFunctionCall_17 {
   class=MaterialExpressionMaterialFunctionCall
   inputs [
    UVs {from=MaterialExpressionMaterialFunctionCall_1, output=Rotated Values}
    Slant amount {from=MaterialExpressionAppendVector_4}
   ]
   properties {materialFunction=/FNE_UI/Systems/_MaterialFunctions/Effects/MF_FNEUI_Slant}
  }
  MaterialExpressionMaterialFunctionCall_18 {
   class=MaterialExpressionMaterialFunctionCall
   inputs [
    UVs {from=MaterialExpressionReroute_4}
    Radius {from=MaterialExpressionReroute_0}
    Glow Min {from=""}
    Glow Max {from=MaterialExpressionReroute_3}
    Corner roundness {from=MaterialExpressionReroute_2}
    Stroke thickness {from=MaterialExpressionReroute_5}
    Is stroke thickness relative to size {from=""}
   ]
   properties {materialFunction=/FNE_UI/Systems/_MaterialFunctions/SDFs/MF_UI_SDF_Triangle}
  }
  MaterialExpressionMaterialFunctionCall_19 {
   class=MaterialExpressionMaterialFunctionCall
   inputs [
    UVs {from=MaterialExpressionReroute_1}
    Radius {from=MaterialExpressionReroute_6}
    Glow Min {from=""}
    Glow Max {from=""}
    Corner roundness {from=MaterialExpressionReroute_7}
    Stroke thickness {from=""}
    Is stroke thickness relative to size {from=""}
   ]
   properties {materialFunction=&MF_UI_SDF_Triangle}
  }
  MaterialExpressionMaterialFunctionCall_2 {
   class=MaterialExpressionMaterialFunctionCall
   inputs [
    In {from=MaterialExpressionNamedRerouteUsage_17}
    Value 1 {from=MaterialExpressionVectorParameter_5, output=RGB}
    Value 2 {from=MaterialExpressionVectorParameter_6, output=RGB}
    Value 3 {from=MaterialExpressionVectorParameter_2, output=RGB}
    Value 4 {from=MaterialExpressionVectorParameter_2, output=RGB}
    Value 5 {from=MaterialExpressionVectorParameter_2, output=RGB}
    Pos value 1 {from=MaterialExpressionScalarParameter_25}
    Pos value 2 {from=MaterialExpressionScalarParameter_27}
    Pos value 3 {from=MaterialExpressionScalarParameter_28}
    Pos value 4 {from=MaterialExpressionScalarParameter_28}
    Pos value 5 {from=MaterialExpressionScalarParameter_28}
   ]
   properties {materialFunction=/FNE_UI/Systems/_MaterialFunctions/Utilities/MF_UI_RemappableGradient}
  }
  MaterialExpressionMaterialFunctionCall_20 {
   class=MaterialExpressionMaterialFunctionCall
   inputs [
    UVs {from=MaterialExpressionNamedRerouteUsage_13}
    Rotation Center {from=MaterialExpressionConstant_0}
    "Rotation Angle (0-1)" {from=MaterialExpressionScalarParameter_19}
   ]
   properties {materialFunction=&CustomRotator}
  }
  MaterialExpressionMaterialFunctionCall_3 {
   class=MaterialExpressionMaterialFunctionCall
   inputs [UV Channel {from=MaterialExpressionMaterialFunctionCall_20, output=Rotated Values}]
   properties {materialFunction=&LinearGradient}
  }
  MaterialExpressionMaterialFunctionCall_4 {
   class=MaterialExpressionMaterialFunctionCall
   inputs [
    UVs {from=""}
    Scale {from=MaterialExpressionAppendVector_0}
    Scale center {from=""}
   ]
   properties {materialFunction=/FNE_UI/Systems/_MaterialFunctions/Transforms/MF_UI_Scale}
  }
  MaterialExpressionMaterialFunctionCall_5 {
   class=MaterialExpressionMaterialFunctionCall
   inputs [UV Channel {from=MaterialExpressionNamedRerouteUsage_10}]
   properties {materialFunction=&LinearGradient}
  }
  MaterialExpressionMaterialFunctionCall_6 {
   class=MaterialExpressionMaterialFunctionCall
   inputs [
    UVs {from=MaterialExpressionNamedRerouteUsage_13}
    CenterPosition {from=""}
    Radius {from=MaterialExpressionScalarParameter_22}
    Density {from=""}
    Invert Density {from=""}
   ]
   properties {materialFunction=&RadialGradientExponential}
  }
  MaterialExpressionMaterialFunctionCall_7 {
   class=MaterialExpressionMaterialFunctionCall
   inputs [
    Input UVs {from=MaterialExpressionMaterialFunctionCall_4, output=Result}
    Translate {from=MaterialExpressionAppendVector_1}
   ]
   properties {materialFunction=/FNE_UI/Systems/_MaterialFunctions/Transforms/MF_UI_Translate}
  }
  MaterialExpressionMaterialFunctionCall_8 {
   class=MaterialExpressionMaterialFunctionCall
   inputs [
    In {from=MaterialExpressionMaterialFunctionCall_5, output=UGradient}
    Value 1 {from=MaterialExpressionVectorParameter_0, output=RGB}
    Value 2 {from=MaterialExpressionVectorParameter_4, output=RGB}
    Value 3 {from=MaterialExpressionVectorParameter_3, output=RGB}
    Value 4 {from=MaterialExpressionVectorParameter_3, output=RGB}
    Value 5 {from=MaterialExpressionVectorParameter_3, output=RGB}
    Pos value 1 {from=MaterialExpressionScalarParameter_29}
    Pos value 2 {from=MaterialExpressionScalarParameter_30}
    Pos value 3 {from=MaterialExpressionScalarParameter_31}
    Pos value 4 {from=MaterialExpressionScalarParameter_31}
    Pos value 5 {from=MaterialExpressionScalarParameter_31}
   ]
   properties {materialFunction=&MF_UI_RemappableGradient}
  }
  MaterialExpressionMaterialFunctionCall_9 {
   class=MaterialExpressionMaterialFunctionCall
   inputs [
    Input UVs {from=""}
    Translate {from=MaterialExpressionAppendVector_2}
   ]
   properties {materialFunction=&MF_UI_Translate}
  }
  MaterialExpressionMax_14 {
   class=MaterialExpressionMax
   inputs [
    A {from=MaterialExpressionNamedRerouteUsage_55}
    B {from=MaterialExpressionNamedRerouteUsage_11}
   ]
  }
  MaterialExpressionMax_15 {
   class=MaterialExpressionMax
   inputs [
    A {from=MaterialExpressionNamedRerouteUsage_15}
    B {from=MaterialExpressionNamedRerouteUsage_16}
   ]
  }
  MaterialExpressionMax_16 {
   class=MaterialExpressionMax
   inputs [
    A {from=MaterialExpressionMin_6}
    B {from=MaterialExpressionMax_15}
   ]
  }
  MaterialExpressionMax_17 {
   class=MaterialExpressionMax
   inputs [
    A {from=MaterialExpressionMax_16}
    B {from=MaterialExpressionMultiply_5}
   ]
  }
  MaterialExpressionMax_18 {
   class=MaterialExpressionMax
   inputs [
    A {from=MaterialExpressionLinearInterpolate_20}
    B {from=MaterialExpressionNamedRerouteUsage_72}
   ]
  }
  MaterialExpressionMin_6 {
   class=MaterialExpressionMin
   inputs [
    A {from=MaterialExpressionNamedRerouteUsage_12, output=""}
    B {from=MaterialExpressionMax_14, output=""}
   ]
   properties {}
  }
  MaterialExpressionMin_7 {
   class=MaterialExpressionMin
   inputs [
    A {from=MaterialExpressionNamedRerouteUsage_61, output=""}
    B {from=MaterialExpressionNamedRerouteUsage_20, output=""}
   ]
   properties {}
  }
  MaterialExpressionMin_8 {
   class=MaterialExpressionMin
   inputs [
    A {from=MaterialExpressionOneMinus_2, output=""}
    B {from=MaterialExpressionNamedRerouteUsage_14, output=""}
   ]
   properties {}
  }
  MaterialExpressionMultiply_0 {
   class=MaterialExpressionMultiply
   inputs [
    A {from=MaterialExpressionNamedRerouteUsage_3}
    B {from=MaterialExpressionScalarParameter_47}
   ]
  }
  MaterialExpressionMultiply_1 {
   class=MaterialExpressionMultiply
   inputs [
    A {from=MaterialExpressionTextureSample_0, output=RGB}
    B {from=MaterialExpressionTextureSample_0, output=RGB}
   ]
  }
  MaterialExpressionMultiply_2 {
   class=MaterialExpressionMultiply
   inputs [
    A {from=MaterialExpressionNamedRerouteUsage_7}
    B {from=MaterialExpressionScalarParameter_48}
   ]
  }
  MaterialExpressionMultiply_3 {
   class=MaterialExpressionMultiply
   inputs [
    A {from=MaterialExpressionNamedRerouteUsage_9}
    B {from=MaterialExpressionScalarParameter_15}
   ]
  }
  MaterialExpressionMultiply_4 {
   class=MaterialExpressionMultiply
   inputs [
    A {from=MaterialExpressionNamedRerouteUsage_52}
    B {from=MaterialExpressionConstant3Vector_0}
   ]
  }
  MaterialExpressionMultiply_5 {
   class=MaterialExpressionMultiply
   inputs [
    A {from=MaterialExpressionNamedRerouteUsage_60}
    B {from=""}
   ]
   properties {constB=0.3}
  }
  MaterialExpressionMultiply_6 {
   class=MaterialExpressionMultiply
   inputs [
    A {from=MaterialExpressionNamedRerouteUsage_66}
    B {from=MaterialExpressionNamedRerouteUsage_67}
   ]
  }
  MaterialExpressionMultiply_7 {
   class=MaterialExpressionMultiply
   inputs [
    A {from=MaterialExpressionLinearInterpolate_1}
    B {from=MaterialExpressionNamedRerouteUsage_37}
   ]
  }
  MaterialExpressionMultiply_8 {
   class=MaterialExpressionMultiply
   inputs [
    A {from=MaterialExpressionLinearInterpolate_2}
    B {from=MaterialExpressionNamedRerouteUsage_38}
   ]
  }
  MaterialExpressionNamedRerouteDeclaration_0 {
   class=MaterialExpressionNamedRerouteDeclaration
   inputs [None {from=MaterialExpressionTextureSample_0, output=A}]
   properties {name=TextureAlphaChannel, nodeColor=<rgba 1 0 0.1875 1>}
  }
  MaterialExpressionNamedRerouteDeclaration_1 {
   class=MaterialExpressionNamedRerouteDeclaration
   inputs [None {from=MaterialExpressionMaterialFunctionCall_0, output=RGBA}]
   properties {name=ShapeCornerRadius, nodeColor=<rgba 0 1 0.3672 1>}
  }
  MaterialExpressionNamedRerouteDeclaration_10 {
   class=MaterialExpressionNamedRerouteDeclaration
   inputs [None {from=MaterialExpressionTextureCoordinate_0}]
   properties {name=UV, nodeColor=<rgba 0 0.5312 1 1>}
  }
  MaterialExpressionNamedRerouteDeclaration_11 {
   class=MaterialExpressionNamedRerouteDeclaration
   inputs [None {from=MaterialExpressionSmoothStep_3}]
   properties {name=FillUV, nodeColor=<rgba 1 0 0.9375 1>}
  }
  MaterialExpressionNamedRerouteDeclaration_12 {
   class=MaterialExpressionNamedRerouteDeclaration
   inputs [None {from=MaterialExpressionSwitch_2}]
   properties {name=GapAlpha, nodeColor=<rgba 1 0 0.0703 1>}
  }
  MaterialExpressionNamedRerouteDeclaration_13 {
   class=MaterialExpressionNamedRerouteDeclaration
   inputs [None {from=MaterialExpressionScalarParameter_14}]
   properties {name=GapThickness, nodeColor=<rgba 0 0.3906 1 1>}
  }
  MaterialExpressionNamedRerouteDeclaration_14 {
   class=MaterialExpressionNamedRerouteDeclaration
   inputs [None {from=MaterialExpressionMultiply_8}]
   properties {name=OutlineColor, nodeColor=<rgba 0 1 0.4141 1>}
  }
  MaterialExpressionNamedRerouteDeclaration_15 {
   class=MaterialExpressionNamedRerouteDeclaration
   inputs [None {from=MaterialExpressionMultiply_7}]
   properties {name=FillColor, nodeColor=<rgba 1 0.0938 0 1>}
  }
  MaterialExpressionNamedRerouteDeclaration_16 {
   class=MaterialExpressionNamedRerouteDeclaration
   inputs [None {from=MaterialExpressionLinearInterpolate_9}]
   properties {name=FillColorUV, nodeColor=<rgba 0 1 0.7188 1>}
  }
  MaterialExpressionNamedRerouteDeclaration_17 {
   class=MaterialExpressionNamedRerouteDeclaration
   inputs [None {from=MaterialExpressionTextureSampleParameter2D_3, output=RGB}]
   properties {name=CustomFillTextureRGB, nodeColor=<rgba 0 1 0.6016 1>}
  }
  MaterialExpressionNamedRerouteDeclaration_18 {
   class=MaterialExpressionNamedRerouteDeclaration
   inputs [None {from=MaterialExpressionMultiply_2}]
   properties {name=OutlineAlpha, nodeColor=<rgba 0 0.8359 1 1>}
  }
  MaterialExpressionNamedRerouteDeclaration_19 {
   class=MaterialExpressionNamedRerouteDeclaration
   inputs [None {from=MaterialExpressionMaterialFunctionCall_11, output=Fill}]
   properties {name=MaskUV, nodeColor=<rgba 0 1 0.0156 1>}
  }
  MaterialExpressionNamedRerouteDeclaration_2 {
   class=MaterialExpressionNamedRerouteDeclaration
   inputs [None {from=MaterialExpressionScalarParameter_1}]
   properties {name=Size, nodeColor=<rgba 1 0.0469 0 1>}
  }
  MaterialExpressionNamedRerouteDeclaration_20 {
   class=MaterialExpressionNamedRerouteDeclaration
   inputs [None {from=MaterialExpressionLinearInterpolate_17}]
   properties {name=FinalAlpha, nodeColor=<rgba 1 0.0703 0 1>}
  }
  MaterialExpressionNamedRerouteDeclaration_21 {
   class=MaterialExpressionNamedRerouteDeclaration
   inputs [None {from=MaterialExpressionMultiply_4}]
   properties {name=MaskRGB, nodeColor=<rgba 1 0 0.2578 1>}
  }
  MaterialExpressionNamedRerouteDeclaration_22 {
   class=MaterialExpressionNamedRerouteDeclaration
   inputs [None {from=MaterialExpressionMin_6}]
   properties {name=TextureAndMaskAlpha, nodeColor=<rgba 1 0 0.8672 1>}
  }
  MaterialExpressionNamedRerouteDeclaration_23 {
   class=MaterialExpressionNamedRerouteDeclaration
   inputs [None {from=MaterialExpressionLinearInterpolate_21}]
   properties {name=FinalColor, nodeColor=<rgba 0 1 0.8125 1>}
  }
  MaterialExpressionNamedRerouteDeclaration_24 {
   class=MaterialExpressionNamedRerouteDeclaration
   inputs [None {from=MaterialExpressionMultiply_3}]
   properties {name=TextureAlpha, nodeColor=<rgba 1 0.8438 0 1>}
  }
  MaterialExpressionNamedRerouteDeclaration_3 {
   class=MaterialExpressionNamedRerouteDeclaration
   inputs [None {from=MaterialExpressionDivide_1}]
   properties {name=Rotation, nodeColor=<rgba 1 0 0.6094 1>}
  }
  MaterialExpressionNamedRerouteDeclaration_4 {
   class=MaterialExpressionNamedRerouteDeclaration
   inputs [None {from=MaterialExpressionMultiply_0}]
   properties {name=FillAlpha, nodeColor=<rgba 1 0 0.5156 1>}
  }
  MaterialExpressionNamedRerouteDeclaration_5 {
   class=MaterialExpressionNamedRerouteDeclaration
   inputs [None {from=MaterialExpressionScalarParameter_7}]
   properties {name=Glow, nodeColor=<rgba 1 0 0.6797 1>}
  }
  MaterialExpressionNamedRerouteDeclaration_6 {
   class=MaterialExpressionNamedRerouteDeclaration
   inputs [None {from=MaterialExpressionMaterialFunctionCall_17, output=Slanted UVs}]
   properties {name=UVwithRotateSlant, nodeColor=<rgba 1 0 0.5391 1>}
  }
  MaterialExpressionNamedRerouteDeclaration_7 {
   class=MaterialExpressionNamedRerouteDeclaration
   inputs [None {from=MaterialExpressionMultiply_1}]
   properties {name=TextureRGB, nodeColor=<rgba 0.2656 1 0 1>}
  }
  MaterialExpressionNamedRerouteDeclaration_8 {
   class=MaterialExpressionNamedRerouteDeclaration
   inputs [None {from=MaterialExpressionScalarParameter_8}]
   properties {name=OutlineThickness, nodeColor=<rgba 1 0 0.5156 1>}
  }
  MaterialExpressionNamedRerouteDeclaration_9 {
   class=MaterialExpressionNamedRerouteDeclaration
   inputs [None {from=MaterialExpressionSwitch_1}]
   properties {name=OutlineUV, nodeColor=<rgba 0 0.8359 1 1>}
  }
  MaterialExpressionNamedRerouteUsage_0 {class=MaterialExpressionNamedRerouteUsage}
  MaterialExpressionNamedRerouteUsage_1 {class=MaterialExpressionNamedRerouteUsage}
  MaterialExpressionNamedRerouteUsage_10 {class=MaterialExpressionNamedRerouteUsage}
  MaterialExpressionNamedRerouteUsage_11 {class=MaterialExpressionNamedRerouteUsage}
  MaterialExpressionNamedRerouteUsage_12 {class=MaterialExpressionNamedRerouteUsage}
  MaterialExpressionNamedRerouteUsage_13 {class=MaterialExpressionNamedRerouteUsage}
  MaterialExpressionNamedRerouteUsage_14 {class=MaterialExpressionNamedRerouteUsage}
  MaterialExpressionNamedRerouteUsage_15 {class=MaterialExpressionNamedRerouteUsage}
  MaterialExpressionNamedRerouteUsage_16 {class=MaterialExpressionNamedRerouteUsage}
  MaterialExpressionNamedRerouteUsage_17 {class=MaterialExpressionNamedRerouteUsage}
  MaterialExpressionNamedRerouteUsage_18 {class=MaterialExpressionNamedRerouteUsage}
  MaterialExpressionNamedRerouteUsage_19 {class=MaterialExpressionNamedRerouteUsage}
  MaterialExpressionNamedRerouteUsage_2 {class=MaterialExpressionNamedRerouteUsage}
  MaterialExpressionNamedRerouteUsage_20 {class=MaterialExpressionNamedRerouteUsage}
  MaterialExpressionNamedRerouteUsage_24 {class=MaterialExpressionNamedRerouteUsage}
  MaterialExpressionNamedRerouteUsage_26 {class=MaterialExpressionNamedRerouteUsage}
  MaterialExpressionNamedRerouteUsage_3 {class=MaterialExpressionNamedRerouteUsage}
  MaterialExpressionNamedRerouteUsage_37 {class=MaterialExpressionNamedRerouteUsage}
  MaterialExpressionNamedRerouteUsage_38 {class=MaterialExpressionNamedRerouteUsage}
  MaterialExpressionNamedRerouteUsage_4 {class=MaterialExpressionNamedRerouteUsage}
  MaterialExpressionNamedRerouteUsage_5 {class=MaterialExpressionNamedRerouteUsage}
  MaterialExpressionNamedRerouteUsage_51 {class=MaterialExpressionNamedRerouteUsage}
  MaterialExpressionNamedRerouteUsage_52 {class=MaterialExpressionNamedRerouteUsage}
  MaterialExpressionNamedRerouteUsage_55 {class=MaterialExpressionNamedRerouteUsage}
  MaterialExpressionNamedRerouteUsage_6 {class=MaterialExpressionNamedRerouteUsage}
  MaterialExpressionNamedRerouteUsage_60 {class=MaterialExpressionNamedRerouteUsage}
  MaterialExpressionNamedRerouteUsage_61 {class=MaterialExpressionNamedRerouteUsage}
  MaterialExpressionNamedRerouteUsage_63 {class=MaterialExpressionNamedRerouteUsage}
  MaterialExpressionNamedRerouteUsage_64 {class=MaterialExpressionNamedRerouteUsage}
  MaterialExpressionNamedRerouteUsage_66 {class=MaterialExpressionNamedRerouteUsage}
  MaterialExpressionNamedRerouteUsage_67 {class=MaterialExpressionNamedRerouteUsage}
  MaterialExpressionNamedRerouteUsage_69 {class=MaterialExpressionNamedRerouteUsage}
  MaterialExpressionNamedRerouteUsage_7 {class=MaterialExpressionNamedRerouteUsage}
  MaterialExpressionNamedRerouteUsage_70 {class=MaterialExpressionNamedRerouteUsage}
  MaterialExpressionNamedRerouteUsage_72 {class=MaterialExpressionNamedRerouteUsage}
  MaterialExpressionNamedRerouteUsage_73 {class=MaterialExpressionNamedRerouteUsage}
  MaterialExpressionNamedRerouteUsage_75 {class=MaterialExpressionNamedRerouteUsage}
  MaterialExpressionNamedRerouteUsage_8 {class=MaterialExpressionNamedRerouteUsage}
  MaterialExpressionNamedRerouteUsage_9 {class=MaterialExpressionNamedRerouteUsage}
  MaterialExpressionOneMinus_2 {class=MaterialExpressionOneMinus, inputs [None {from=MaterialExpressionNamedRerouteUsage_70}]}
  MaterialExpressionReroute_0 {class=MaterialExpressionReroute, inputs [None {from=MaterialExpressionNamedRerouteUsage_4}]}
  MaterialExpressionReroute_1 {class=MaterialExpressionReroute, inputs [None {from=MaterialExpressionNamedRerouteUsage_6}]}
  MaterialExpressionReroute_2 {class=MaterialExpressionReroute, inputs [None {from=MaterialExpressionNamedRerouteUsage_5}]}
  MaterialExpressionReroute_3 {class=MaterialExpressionReroute, inputs [None {from=MaterialExpressionNamedRerouteUsage_8}]}
  MaterialExpressionReroute_4 {class=MaterialExpressionReroute, inputs [None {from=MaterialExpressionNamedRerouteUsage_1}]}
  MaterialExpressionReroute_5 {class=MaterialExpressionReroute, inputs [None {from=MaterialExpressionNamedRerouteUsage_2}]}
  MaterialExpressionReroute_6 {class=MaterialExpressionReroute, inputs [None {from=MaterialExpressionLinearInterpolate_4}]}
  MaterialExpressionReroute_7 {class=MaterialExpressionReroute, inputs [None {from=MaterialExpressionNamedRerouteUsage_19}]}
  MaterialExpressionReroute_8 {class=MaterialExpressionReroute, inputs [None {from=MaterialExpressionNamedRerouteUsage_18}]}
  MaterialExpressionReroute_9 {class=MaterialExpressionReroute, inputs [None {from=MaterialExpressionReroute_8}]}
  MaterialExpressionScalarParameter_0 {
   class=MaterialExpressionScalarParameter
   properties {desc="Rotation of the shape, normalized to 360 degrees", group="[2] Fill", parameterName=ShapeRotation, sliderMax=360}
  }
  MaterialExpressionScalarParameter_1 {class=MaterialExpressionScalarParameter, properties {defaultValue=0.8, group="[2] Fill", parameterName=ShapeSize, sliderMax=1}}
  MaterialExpressionScalarParameter_10 {
   class=MaterialExpressionScalarParameter
   properties {
    desc=Sets the bottom left corner radius of the shape and outline.
    group="[2] Fill"
    parameterName=BottomLeftCornerRadius
    sliderMax=0.5
    sortPriority=33
   }
  }
  MaterialExpressionScalarParameter_11 {
   class=MaterialExpressionScalarParameter
   properties {
    desc=Sets the bottom right corner radius of the shape and outline.
    group="[2] Fill"
    parameterName=BottomRightCornerRadius
    sliderMax=0.5
    sortPriority=33
   }
  }
  MaterialExpressionScalarParameter_12 {
   class=MaterialExpressionScalarParameter
   properties {
    desc=Sets the top right corner radius of the shape and outline.
    group="[2] Fill"
    parameterName=TopRightCornerRadius
    sliderMax=0.5
    sortPriority=33
   }
  }
  MaterialExpressionScalarParameter_13 {
   class=MaterialExpressionScalarParameter
   properties {desc=Sets the amount of horizontal slant for the fill., group="[5] Slant", parameterName=SlantX, sliderMax=1, sliderMin=-1}
  }
  MaterialExpressionScalarParameter_14 {
   class=MaterialExpressionScalarParameter
   properties {
    defaultValue=0.1
    desc=Sets the thickness of the gap between the shape and outline.
    group="[3] Outline"
    parameterName=GapThickness
    sliderMax=1
   }
  }
  MaterialExpressionScalarParameter_15 {
   class=MaterialExpressionScalarParameter
   properties {defaultValue=1, desc=Sets the transparency of the texture., group="[1] Texture", parameterName=TextureAlpha, sliderMax=1}
  }
  MaterialExpressionScalarParameter_16 {
   class=MaterialExpressionScalarParameter
   properties {defaultValue=1, desc=Sets the transparency of the fill color., group="[2] Fill", parameterName=FillAlpha, sliderMax=1}
  }
  MaterialExpressionScalarParameter_17 {
   class=MaterialExpressionScalarParameter
   properties {desc=Sets the transparency of the outline., group="[3] Outline", parameterName=OutlineAlpha, sliderMax=1}
  }
  MaterialExpressionScalarParameter_18 {
   class=MaterialExpressionScalarParameter
   properties {desc=Sets the amount of vertical slant for the fill., group="[5] Slant", parameterName=SlantY, sliderMax=1, sliderMin=-1}
  }
  MaterialExpressionScalarParameter_19 {
   class=MaterialExpressionScalarParameter
   properties {
    desc="Rotation of the linear gradient fill color.\r\n\r\nONLY FOR LINEAR GRADIENT FILL"
    group="[2] Fill"
    parameterName=LinearGradientFillRotation
    sliderMax=1
   }
  }
  MaterialExpressionScalarParameter_2 {
   class=MaterialExpressionScalarParameter
   properties {
    desc="Sets the horizontal position of the icon. Negative numbers move left, positive numbers move right."
    group="[1] Texture"
    parameterName=TexturePositionX
    sliderMax=1
    sliderMin=-1
   }
  }
  MaterialExpressionScalarParameter_20 {
   class=MaterialExpressionScalarParameter
   properties {desc=Sets the transparency of the outline., group="[3] Outline", parameterName=OutlineAlpha, sliderMax=1}
  }
  MaterialExpressionScalarParameter_22 {
   class=MaterialExpressionScalarParameter
   properties {
    defaultValue=0.4
    desc="Sets the size of the radial gradient in the fill color.\r\n\r\nONLY FOR RADIAL GRADIENT."
    group="[2] Fill"
    parameterName=FillColorRadialGradientSize
   }
  }
  MaterialExpressionScalarParameter_23 {
   class=MaterialExpressionScalarParameter
   properties {
    desc="Sets the gradient type for the fill color.\r\n0.0 = Circle Gradient\r\n1.0 = Linear Gradient"
    group="[2] Fill"
    parameterName=FillGradientType
   }
  }
  MaterialExpressionScalarParameter_24 {
   class=MaterialExpressionScalarParameter
   properties {desc="Sets the fill shape.\r\n0.0 for Box\r\n1.0 for Circle\r\n2.0 for Hexagon\r\n3.0 for Triangle", group="[2] Fill", parameterName=Shape}
  }
  MaterialExpressionScalarParameter_25 {
   class=MaterialExpressionScalarParameter
   properties {desc=Sets the position of the first color in the fill gradient., group="[2] Fill", parameterName=FillColor1GradientPostion}
  }
  MaterialExpressionScalarParameter_27 {
   class=MaterialExpressionScalarParameter
   properties {
    defaultValue=0.5
    desc=Sets the position of the second color in the fill gradient.
    group="[2] Fill"
    parameterName=FillColor2GradientPostion
   }
  }
  MaterialExpressionScalarParameter_28 {
   class=MaterialExpressionScalarParameter
   properties {defaultValue=1, desc=Sets the position of the third color in the fill gradient., group="[2] Fill", parameterName=FillColor3GradientPostion}
  }
  MaterialExpressionScalarParameter_29 {
   class=MaterialExpressionScalarParameter
   properties {
    desc=Sets the position of the first color in the outline gradient.
    group="[3] Outline"
    parameterName=OutlineColor1GradientPosition
    sliderMax=1
   }
  }
  MaterialExpressionScalarParameter_3 {
   class=MaterialExpressionScalarParameter
   properties {defaultValue=1, desc=Sets the horizontal scale of the icon., group="[1] Texture", parameterName=TextureScaleX, sliderMax=1}
  }
  MaterialExpressionScalarParameter_30 {
   class=MaterialExpressionScalarParameter
   properties {
    defaultValue=0.5
    desc=Sets the position of the second color in the outline gradient.
    group="[3] Outline"
    parameterName=OutlineColor2GradientPosition
    sliderMax=1
   }
  }
  MaterialExpressionScalarParameter_31 {
   class=MaterialExpressionScalarParameter
   properties {
    defaultValue=1
    desc=Sets the position of the third color in the outline gradient.
    group="[3] Outline"
    parameterName=OutlineColor3GradientPosition
    sliderMax=1
   }
  }
  MaterialExpressionScalarParameter_33 {
   class=MaterialExpressionScalarParameter
   properties {desc="Sets the vertical position of the texture mask. ", group="[4] Texture Mask", parameterName=MaskPositionY, sliderMax=1, sliderMin=-1}
  }
  MaterialExpressionScalarParameter_34 {
   class=MaterialExpressionScalarParameter
   properties {desc="Sets the horizontal position of the texture mask. ", group="[4] Texture Mask", parameterName=MaskPositionX, sliderMax=1, sliderMin=-1}
  }
  MaterialExpressionScalarParameter_35 {
   class=MaterialExpressionScalarParameter
   properties {desc=Sets the width of the texture mask., group="[4] Texture Mask", parameterName=MaskWidth, sliderMax=1, sliderMin=-1, sortPriority=34}
  }
  MaterialExpressionScalarParameter_36 {
   class=MaterialExpressionScalarParameter
   properties {desc="Sets the height of the texture mask. ", group="[4] Texture Mask", parameterName=MaskHeight, sliderMax=1, sliderMin=-1, sortPriority=33}
  }
  MaterialExpressionScalarParameter_4 {
   class=MaterialExpressionScalarParameter
   properties {defaultValue=1, desc=Sets the vertical scale of the icon., group="[1] Texture", parameterName=TextureScaleY, sliderMax=1}
  }
  MaterialExpressionScalarParameter_41 {
   class=MaterialExpressionScalarParameter
   properties {desc="[DEBUG] Previews the texture mask.", group="[4] Texture Mask", parameterName=ShowMaskPreview, sliderMax=1, sortPriority=0}
  }
  MaterialExpressionScalarParameter_42 {
   class=MaterialExpressionScalarParameter
   properties {desc="[DEBUG] Previews the texture mask.", group="[4] Texture Mask", parameterName=ShowMaskPreview, sliderMax=1, sortPriority=0}
  }
  MaterialExpressionScalarParameter_47 {
   class=MaterialExpressionScalarParameter
   properties {defaultValue=1, desc=Sets the transparency of the fill color., group="[2] Fill", parameterName=FillAlpha, sliderMax=1}
  }
  MaterialExpressionScalarParameter_48 {
   class=MaterialExpressionScalarParameter
   properties {desc=Sets the transparency of the outline., group="[3] Outline", parameterName=OutlineAlpha, sliderMax=1}
  }
  MaterialExpressionScalarParameter_5 {
   class=MaterialExpressionScalarParameter
   properties {
    desc="Sets the vertical position of the icon. Negative numbers move down, positive numbers move up."
    group="[1] Texture"
    parameterName=TexturePositionY
    sliderMax=1
    sliderMin=-1
   }
  }
  MaterialExpressionScalarParameter_6 {
   class=MaterialExpressionScalarParameter
   properties {desc="Sets the fill shape.\r\n0.0 for Box\r\n1.0 for Circle\r\n2.0 for Hexagon\r\n3.0 for Triangle", group="[2] Fill", parameterName=Shape}
  }
  MaterialExpressionScalarParameter_7 {class=MaterialExpressionScalarParameter, properties {group="[2] Fill", parameterName=Glow, sliderMax=1}}
  MaterialExpressionScalarParameter_8 {class=MaterialExpressionScalarParameter, properties {defaultValue=0.1, group="[3] Outline", parameterName=OutlineThickness, sliderMax=1}}
  MaterialExpressionScalarParameter_9 {
   class=MaterialExpressionScalarParameter
   properties {
    desc=Sets the top left corner radius of the shape and outline.
    group="[2] Fill"
    parameterName=TopLeftCornerRadius
    sliderMax=0.5
    sortPriority=33
   }
  }
  MaterialExpressionSmoothStep_3 {
   class=MaterialExpressionSmoothStep
   inputs [
    Min {from="", output=""}
    Max {from=MaterialExpressionSubtract_9, output=""}
    Value {from=MaterialExpressionNamedRerouteUsage_24, output=""}
   ]
   properties {}
  }
  MaterialExpressionSubtract_14 {
   class=MaterialExpressionSubtract
   inputs [
    A {from=MaterialExpressionNamedRerouteUsage_18}
    B {from=MaterialExpressionNamedRerouteUsage_26}
   ]
  }
  MaterialExpressionSubtract_9 {
   class=MaterialExpressionSubtract
   inputs [
    A {from=MaterialExpressionSwitch_0}
    B {from=MaterialExpressionNamedRerouteDeclaration_9}
   ]
  }
  MaterialExpressionSwitch_0 {
   class=MaterialExpressionSwitch
   inputs [
    SwitchValue {from=MaterialExpressionScalarParameter_6, output=""}
    Default {from="", output=""}
    None {from=MaterialExpressionMaterialFunctionCall_10, output=Fill}
    None {from=MaterialExpressionMaterialFunctionCall_12, output=Fill}
    None {from=MaterialExpressionMaterialFunctionCall_13, output=Fill}
    None {from=MaterialExpressionMaterialFunctionCall_18, output=Fill}
   ]
   properties {inputs [
     {
      inputName=None
      input {
       expression=&self:MaterialExpressionMaterialFunctionCall_10
       inputName=None
       mask=0
       maskA=0
       maskB=0
       maskG=0
       maskR=0
       outputIndex=0
      }
     }
     {
      inputName=None
      input {
       expression=&self:MaterialExpressionMaterialFunctionCall_12
       inputName=None
       mask=0
       maskA=0
       maskB=0
       maskG=0
       maskR=0
       outputIndex=0
      }
     }
     {
      inputName=None
      input {
       expression=&self:MaterialExpressionMaterialFunctionCall_13
       inputName=None
       mask=0
       maskA=0
       maskB=0
       maskG=0
       maskR=0
       outputIndex=0
      }
     }
     {
      inputName=None
      input {
       expression=&self:MaterialExpressionMaterialFunctionCall_18
       inputName=None
       mask=0
       maskA=0
       maskB=0
       maskG=0
       maskR=0
       outputIndex=0
      }
     }
    ]}
  }
  MaterialExpressionSwitch_1 {
   class=MaterialExpressionSwitch
   inputs [
    SwitchValue {from=MaterialExpressionScalarParameter_6, output=""}
    Default {from="", output=""}
    None {from=MaterialExpressionMaterialFunctionCall_10, output=Outline}
    None {from=MaterialExpressionMaterialFunctionCall_12, output=Outline}
    None {from=MaterialExpressionMaterialFunctionCall_13, output=Outline}
    None {from=MaterialExpressionMaterialFunctionCall_18, output=Outline}
   ]
   properties {inputs [
     {
      inputName=None
      input {
       expression=&self:MaterialExpressionMaterialFunctionCall_10
       inputName=None
       mask=0
       maskA=0
       maskB=0
       maskG=0
       maskR=0
       outputIndex=1
      }
     }
     {
      inputName=None
      input {
       expression=&self:MaterialExpressionMaterialFunctionCall_12
       inputName=None
       mask=0
       maskA=0
       maskB=0
       maskG=0
       maskR=0
       outputIndex=1
      }
     }
     {
      inputName=None
      input {
       expression=&self:MaterialExpressionMaterialFunctionCall_13
       inputName=None
       mask=0
       maskA=0
       maskB=0
       maskG=0
       maskR=0
       outputIndex=1
      }
     }
     {
      inputName=None
      input {
       expression=&self:MaterialExpressionMaterialFunctionCall_18
       inputName=None
       mask=0
       maskA=0
       maskB=0
       maskG=0
       maskR=0
       outputIndex=1
      }
     }
    ]}
  }
  MaterialExpressionSwitch_2 {
   class=MaterialExpressionSwitch
   inputs [
    SwitchValue {from=MaterialExpressionScalarParameter_24, output=""}
    Default {from="", output=""}
    None {from=MaterialExpressionMaterialFunctionCall_14, output=Fill}
    None {from=MaterialExpressionMaterialFunctionCall_15, output=Fill}
    None {from=MaterialExpressionMaterialFunctionCall_16, output=Fill}
    None {from=MaterialExpressionMaterialFunctionCall_19, output=Fill}
   ]
   properties {inputs [
     {
      inputName=None
      input {
       expression=&self:MaterialExpressionMaterialFunctionCall_14
       inputName=None
       mask=0
       maskA=0
       maskB=0
       maskG=0
       maskR=0
       outputIndex=0
      }
     }
     {
      inputName=None
      input {
       expression=&self:MaterialExpressionMaterialFunctionCall_15
       inputName=None
       mask=0
       maskA=0
       maskB=0
       maskG=0
       maskR=0
       outputIndex=0
      }
     }
     {
      inputName=None
      input {
       expression=&self:MaterialExpressionMaterialFunctionCall_16
       inputName=None
       mask=0
       maskA=0
       maskB=0
       maskG=0
       maskR=0
       outputIndex=0
      }
     }
     {
      inputName=None
      input {
       expression=&self:MaterialExpressionMaterialFunctionCall_19
       inputName=None
       mask=0
       maskA=0
       maskB=0
       maskG=0
       maskR=0
       outputIndex=0
      }
     }
    ]}
  }
  MaterialExpressionTextureCoordinate_0 {class=MaterialExpressionTextureCoordinate}
  MaterialExpressionTextureObjectParameter_0 {
   class=MaterialExpressionTextureObjectParameter
   inputs [
    UVs {from="", output=""}
    Apply View MipBias {from="", output=""}
   ]
   properties {
    group="[1] Texture"
    parameterName=Texture
    texture=/Game/Athena/UI/InGame/Creative/PropertyEditor/Icon_Picker/Textures_V2/128x128/T_UI_IconLibrary_Llama_128
   }
  }
  MaterialExpressionTextureSampleParameter2D_3 {
   class=MaterialExpressionTextureSampleParameter2D
   inputs [
    UVs {from="", output=""}
    Apply View MipBias {from="", output=""}
   ]
   properties {
    desc="Custom texture for the fill color.\r\n\r\n(Only if CustomFillTexture = 1)"
    group="[2] Fill"
    parameterName=CustomFillTexture
    texture=/EpicBaseTextures/Ground/Manufactured/Japanese_Wooden_Floor/T_Japanese_Wooden_Floor_D
   }
  }
  MaterialExpressionTextureSample_0 {
   class=MaterialExpressionTextureSample
   inputs [
    UVs {from=MaterialExpressionMaterialFunctionCall_7, output=UVs raw}
    Tex {from=MaterialExpressionTextureObjectParameter_0}
    Apply View MipBias {from=""}
   ]
   properties {samplerSource=SSM_Clamp_WorldGroupSettings}
  }
  MaterialExpressionVectorParameter_0 {
   class=MaterialExpressionVectorParameter
   properties {desc=First color in the outline gradient., group="[3] Outline", parameterName=OutlineColor1, defaultValue=<rgba 1 0.0332 0 1>}
  }
  MaterialExpressionVectorParameter_2 {
   class=MaterialExpressionVectorParameter
   properties {desc=Third color in the fill gradient., group="[2] Fill", parameterName=FillColor3, defaultValue=<rgba 0.0103 1 0 1>}
  }
  MaterialExpressionVectorParameter_3 {
   class=MaterialExpressionVectorParameter
   properties {desc=Third color in the outline gradient., group="[3] Outline", parameterName=OutlineColor3, defaultValue=<rgba 0.1069 1 0 1>}
  }
  MaterialExpressionVectorParameter_4 {
   class=MaterialExpressionVectorParameter
   properties {desc=Second color in the outline gradient., group="[3] Outline", parameterName=OutlineColor2, defaultValue=<rgba 1 0.5191 0 1>}
  }
  MaterialExpressionVectorParameter_5 {
   class=MaterialExpressionVectorParameter
   properties {desc=First color in the fill gradient., group="[2] Fill", parameterName=FillColor1, defaultValue=<rgba 0.0065 1 0 1>}
  }
  MaterialExpressionVectorParameter_6 {
   class=MaterialExpressionVectorParameter
   properties {desc=Second color in the fill gradient., group="[2] Fill", parameterName=FillColor2, defaultValue=<rgba 0.0067 1 0 1>}
  }
 }
 properties {bAutomaticallySetUsageInEditor=false, blendMode=BLEND_Translucent, materialDomain=MD_UI, usedShadingModels=Default Lit}
}
